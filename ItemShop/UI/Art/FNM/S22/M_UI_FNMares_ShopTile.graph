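# UI material for the Fortnitemares (FNMares, S22) Item Shop tile background, built in dark, red and gold tones.
# Radial gradients, an arrow mask filled with a rotating gradient, and procedural confetti combine with a background texture.
# Panning swirl flowmaps warp the arrow gradient and confetti UVs.
# Time-stepped scrolling static distorts the frame texture's UVs for a glitch effect, and its inverted result helps drive the background blend.
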Material /ItemShop/UI/Art/FNM/S22/M_UI_FNMares_ShopTile
graph {
 properties {bUsedWithStaticMesh=false, materialDomain=MD_UI, pixelDepthOffsetMode=PDOM_Legacy, usedShadingModels=Default Lit}
 nodes [
  AppendVector(A=ScalarParameter() {parameterName=L1-Patter-X-Offset}, B=ScalarParameter() {defaultValue=0.5, parameterName=L1-Patter-Y-Offset})
  n1 = MaterialFunctionCall() {materialFunction=&GetUserInterfaceUV}
  n2 = Subtract(A=@n1.Pixel Size, B=Multiply(A=ScalarParameter() {defaultValue=6, parameterName=PixelSize}) {constB=2})
  Divide(A=ComponentMask(None=@n2) {r=true}, B=ComponentMask(None=@n1.Pixel Size) {r=true})
  Divide(A=ComponentMask(None=@n2) {g=true}, B=ComponentMask(None=@n1.Pixel Size) {g=true})
  n3 = MaterialFunctionCall() {materialFunction=&MF_UI_UVScaling}
  n4 = ScalarParameter() {defaultValue=1.1583, parameterName=Scale}
  n5 = ScalarParameter() {defaultValue=0.005, parameterName=X}
  n6 = ScalarParameter() {defaultValue=0.192, parameterName=Y}
  n7 = MaterialFunctionCall(
   UV=Add(
    A=MaterialFunctionCall(UVs=@n3.Scaled to Fit, Texture Scale=@n4) {materialFunction=&ScaleUVsByCenter}.UVs
    B=MaterialFunctionCall(X=@n5, Y=@n6) {materialFunction=&MakeFloat2}.Result
   )
   GradientStart=ScalarParameter() {defaultValue=0.029, parameterName=GradientStart}
   GradientEnd=ScalarParameter() {defaultValue=0.7818, parameterName=GradientEnd}
  ) {materialFunction=&MF_UI_RadialGradient}
  LinearInterpolate(Alpha=@n7.Smooth)
  n8 = Add(
   A=MaterialFunctionCall(
    UVs=MaterialFunctionCall() {materialFunction=&MF_UI_UVScaling}.Scaled to Fit
    Texture Scale=ScalarParameter() {defaultValue=0.6005, parameterName=Scale g2}
   ) {materialFunction=&ScaleUVsByCenter}.UVs
   B=MaterialFunctionCall(X=ScalarParameter() {defaultValue=0.005, parameterName=X g2}, Y=ScalarParameter() {defaultValue=-0.056, parameterName=Y g2}) {materialFunction=&MakeFloat2}.Result
  )
  n9 = ScalarParameter() {defaultValue=0.101, parameterName=GradientStart g2}
  n10 = ScalarParameter() {defaultValue=1.3253, parameterName=GradientEnd g2}
  n11 = LinearInterpolate(A=Constant3Vector(), B=Constant3Vector() {constant=<rgba 0.0312 0.0003 0 0>}, Alpha=@n7.Smooth)
  LinearInterpolate(
   A=LinearInterpolate(
    A=Constant3Vector()
    B=Constant3Vector()
    Alpha=MaterialFunctionCall(UV=@n8, GradientStart=@n9, GradientEnd=@n10) {materialFunction=&MF_UI_RadialGradient}.Smooth
   )
   B=@n11
  )
  n12 = Constant3Vector() {constant=<rgba 0 0 0 1>}
  n13 = Constant3Vector() {constant=<rgba 0.7913 0 0 1>}
  n14 = Rotator(Coordinate=TextureCoordinate(), Time=Multiply(A=Time(), B=Constant() {r=1.5}))
  n15 = Constant2Vector() {g=1, r=-1}
  n16 = Panner(Coordinate=TextureCoordinate()) {speedX=0.1, speedY=0.1}
  n17 = TextureObject() {samplerType=SAMPLERTYPE_LinearColor, texture=/Game/Global/Textures/Default/Flowmaps/T_SwirlFlow_02}
  n18 = Multiply(A=Constant() {r=0.5}, B=ComponentMask(None=TextureSample(UVs=@n16, Tex=@n17).R) {g=true, r=true})
  n19 = MaterialFunctionCall(UVs=Add(A=Multiply(A=@n15, B=@n18), B=TextureCoordinate()), Texture Scale=ScalarParameter() {defaultValue=2, parameterName=Layer2Scale}) {materialFunction=&ScaleUVsByCenter}
  n20 = MaterialFunctionCall(
   Texture=Multiply(A=Multiply(A=@n14, B=@n19.UVs), B=ScalarParameter() {defaultValue=7.3891, parameterName=GSize})
   GradientStart=Constant() {r=1}
   GradientEnd=Constant() {r=5}
  ) {materialFunction=&MF_UI_LinearGradient}
  n21 = Add(
   A=MaterialFunctionCall(
    UVs=MaterialFunctionCall() {materialFunction=&MF_UI_UVScaling}.Scaled To Fill
    Texture Scale=ScalarParameter() {defaultValue=1.2, parameterName=Scale_Arrows}
   ) {materialFunction=&ScaleUVsByCenter}.UVs
   B=MaterialFunctionCall(X=ScalarParameter() {parameterName=arrowX}, Y=ScalarParameter() {defaultValue=0.04, parameterName=ArrowY}) {materialFunction=&MakeFloat2}.Result
  )
  n22 = LinearInterpolate(B=Constant3Vector() {constant=<rgba 0.0184 0.0325 0.3177 0>})
  n23 = MaterialFunctionCall() {materialFunction=&MF_UI_UVScaling}
  n24 = ScalarParameter() {defaultValue=1.3802, parameterName=Scale_G2}
  n25 = ScalarParameter() {defaultValue=0.005, parameterName=X_G2}
  n26 = ScalarParameter() {defaultValue=-0.296, parameterName=Y_G2}
  n27 = MaterialFunctionCall(
   UV=Add(
    A=MaterialFunctionCall(UVs=@n23.Scaled to Fit, Texture Scale=@n24) {materialFunction=&ScaleUVsByCenter}.UVs
    B=MaterialFunctionCall(X=@n25, Y=@n26) {materialFunction=&MakeFloat2}.Result
   )
   GradientStart=Constant() {r=-0.008}
   GradientEnd=Constant() {r=0.612}
  ) {materialFunction=&MF_UI_RadialGradient}
  n28 = LinearInterpolate(B=Constant3Vector() {constant=<rgba 0.5521 0 0.1072 0>})
  n29 = Add(
   A=LinearInterpolate(B=LinearInterpolate(A=@n12, B=@n13, Alpha=@n20.Result), Alpha=TextureSample(UVs=@n21) {texture=/OfferCatalog/Art/FNM/S22/T_UI_Arrows}.R)
   B=LinearInterpolate(A=LinearInterpolate(B=@n22, Alpha=@n27.Smooth), B=LinearInterpolate(B=@n28, Alpha=@n27.Smooth)) {constAlpha=0}
  )
  n30 = Constant() {r=0.5}
  n31 = TextureCoordinate()
  n32 = ComponentMask(None=TextureSample(
    UVs=Panner(Coordinate=@n31) {speedX=0.07, speedY=0.0132}
    Tex=TextureObject() {samplerType=SAMPLERTYPE_LinearColor, texture=&T_SwirlFlow_02}
   ).R) {g=true, r=true}
  n33 = Multiply(A=Multiply(A=Constant2Vector() {g=1, r=-1}, B=Multiply(A=@n30, B=@n32)))
  n34 = TextureCoordinate()
  n35 = MaterialFunctionCall!(
   "UV(V2)"=MaterialFunctionCall(UVs=Add(A=@n33, B=@n34), Texture Scale=ScalarParameter() {defaultValue=1.9466, parameterName=Layer1Scale}) {materialFunction=&ScaleUVsByCenter}.UVs
   Global-Angle=ScalarParameter() {defaultValue=-0.5, parameterName=L1-Global Angle}
   "Pattern-Tiles(V2)"=AppendVector(A=ScalarParameter() {defaultValue=15, parameterName=L1-Pattern-X}, B=ScalarParameter() {defaultValue=300, parameterName=L1-Pattern-Y})
   RandomSeed=ScalarParameter() {defaultValue=12345, parameterName=L1-RandomSeed}
   Confetti-Speed-TimeFactor=ScalarParameter() {defaultValue=-0.05, parameterName=L1-Confetti-Speed-TimeFactor}
   Confetti-Rotation-TimeFactor=ScalarParameter() {defaultValue=0.25, parameterName=L1-Confetti-Rotation-TimeFactor}
   Confetti-Local-Rotation-Range=ScalarParameter() {defaultValue=0.0075, parameterName=L1-Confetti-Local-Rotation-Range}
   "Confetti-Local-Tiles(V2)"=AppendVector(
    A=ScalarParameter() {defaultValue=50, parameterName=L1-Confetti-Local-Tile-X}
    B=ScalarParameter() {defaultValue=30, parameterName=L1-Confetti-Local-Tile-Y}
   )
   "SmoothStepRanges(V2)"=AppendVector(
    A=ScalarParameter() {defaultValue=1, parameterName=L1-SmoothStep-Max}
    B=ScalarParameter() {defaultValue=0.5, parameterName=L1-SmoothStep-Min}
   )
  ) {materialFunction=/Game/UI/Frontend/ItemShop/Art/Cactus/MF_UI_Confetti}
  n36 = LinearInterpolate(A=LinearInterpolate(A=@n35.Confetti-Mask, B=Constant3Vector() {constant=<rgba 0.6562 0.4415 0 0>}, Alpha=@n35.Confetti-Mask), B=@n11)
  n37 = TextureSample(UVs=MaterialFunctionCall() {materialFunction=&MF_UI_UVScaling}.Scaled To Fill) {texture=/OfferCatalog/Art/FNM/S22/T_UI_FNMares_BGTexture}
  n38 = VectorParameter() {parameterName=Emissive Color, defaultValue=<rgba 1 1 1 0>}
  n39 = MaterialFunctionCall() {materialFunction=&MF_UI_UVScaling}
  n40 = ScalarParameter() {defaultValue=0.344, group=Static, parameterName=StaticLowValue}
  n41 = ScalarParameter() {defaultValue=1, group=Static, parameterName=StaticHighValue}
  n42 = TextureCoordinate()
  n43 = Multiply(A=Time(), B=ScalarParameter() {defaultValue=0.08, parameterName=BigGlitchSpeed})
  n44 = Round(None=Power(Base=Frac(None=@n43)))
  n45 = LinearInterpolate(A=@n42, B=Multiply(A=@n42) {constB=0.75}, Alpha=@n44)
  n46 = TextureSample(UVs=Panner(Coordinate=@n45) {speedX=5, speedY=0.1}) {samplerType=SAMPLERTYPE_Masks, texture=/Game/Missions/Primary/RepairTheShelter/Textures/Cinematic_Screen_Static}
  n47 = Multiply(
   A=TextureSample(UVs=Panner(Coordinate=@n45) {speedX=6, speedY=-0.05}) {samplerType=SAMPLERTYPE_Masks, texture=&Cinematic_Screen_Static}.G
   B=TextureSample(UVs=Panner(Coordinate=@n45) {speedX=-3, speedY=-0.03}) {samplerType=SAMPLERTYPE_Masks, texture=&Cinematic_Screen_Static}.R
  )
  n48 = Multiply(
   A=Add(A=Multiply(A=@n46.B) {constB=0.1}, B=@n47)
   B=LinearInterpolate(
    A=ScalarParameter() {defaultValue=4.25, group=Static, parameterName=Static Scale}
    B=ScalarParameter() {defaultValue=10, group=Static, parameterName=Static Big Scale}
   )
  )
  n49 = LinearInterpolate(
   A=ScalarParameter() {defaultValue=50, group=Static, parameterName=Static UV Distort}
   B=ScalarParameter() {defaultValue=8, group=Static, parameterName=Static UV Big Distort}
   Alpha=@n44
  )
  n50 = Multiply(A=Add(A=LinearInterpolate(A=@n40, B=@n41, Alpha=@n48)), B=AppendVector(A=Multiply(A=@n47, B=@n49), B=Constant() {r=3}))
  n51 = Multiply(A=TextureSample(UVs=Add(A=@n39.Scaled To Fill, B=@n50)) {texture=/OfferCatalog/Art/Alfredo/T_UI_Frame}.R, B=@n38.RGB)
  n52 = OneMinus(None=LinearInterpolate(A=Multiply(A=@n38.RGB), B=Multiply(A=@n51, B=@n38.RGB)) {constAlpha=0.8})
  LinearInterpolate(B=LinearInterpolate(A=Add(A=@n29, B=@n36), B=@n37.RGB), Alpha=Add(A=@n37.RGB, B=Multiply(A=@n52) {constB=1.5}))
  n53 = ScalarParameter() {defaultValue=0.8377, parameterName=FrameFeather1}
  n54 = ScalarParameter() {defaultValue=1.2794, parameterName=FrameFeather2}
  n55 = Abs(None=ConstantBiasScale(None=MaterialFunctionCall() {materialFunction=&MF_UI_UVScaling}.ScaleY) {bias=-0.5, scale=2})
  n56 = Add(
   A=SmoothStep(Min=@n53, Max=@n54, Value=ComponentMask(None=@n55) {r=true})
   B=SmoothStep(Min=@n53, Max=@n54, Value=ComponentMask(None=@n55) {g=true})
  )
  Multiply(A=Multiply(A=Saturate(None=@n56), B=ScalarParameter() {defaultValue=10, parameterName=FrameOpacity}), B=@n38.RGB)
  TextureCoordinate()
 ]
}
