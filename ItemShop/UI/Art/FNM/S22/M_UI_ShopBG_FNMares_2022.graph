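# UI material for the Fortnitemares 2022 Item Shop background, with animated purple mist and a periodic lighting change.
# Flow-mapped, panning noise is tinted purple, shaped by linear gradients and a mask texture, then blended with the store artwork.
# UV-derived horizontal and vertical falloffs darken the artwork's edges.
# A time-driven sine, remapped by SmoothStep, periodically crossfades the darkened artwork to its lit variant.
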Material /ItemShop/UI/Art/FNM/S22/M_UI_ShopBG_FNMares_2022
graph {
 properties {bUsedWithStaticMesh=false, materialDomain=MD_UI, pixelDepthOffsetMode=PDOM_Legacy, usedShadingModels=Default Lit}
 nodes [
  Add()
  n1 = Time()
  n2 = ScalarParameter() {defaultValue=0.5, parameterName=tenminutes}
  n3 = Multiply(A=Sine(None=Multiply(A=@n1, B=@n2)) {period=10}) {constB=20}
  Ceil(None=@n3)
  Ceil()
  ConstantBiasScale()
  Floor()
  n4 = MaterialFunctionCall() {materialFunction=&MF_UI_UVScaling}
  n5 = MaterialFunctionCall!(
   Texture 2D=TextureObject() {samplerType=SAMPLERTYPE_Grayscale, texture=/Game/UI/Frontend/ItemShop/Art/Rumble/LowResBlurredNoise_Rumble}
   Texture UVs=Panner(Coordinate=Multiply(A=@n4.Scaled To Fill) {constB=0.8942}) {speedY=0.25}
   FlowMap Texture 2D=TextureObject() {samplerType=SAMPLERTYPE_LinearColor, texture=/Game/UI/Frontend/ItemShop/Art/WaterBackground/T_UI_WaterFlowMap}
   FlowMap UVs
   FlowIntensity=Constant() {r=0.7583}
   FlowSpeed
  ) {materialFunction=/FortUILibrary/MaterialFunctions/Effects/MF_UI_FlowMap}
  n6 = Multiply(A=@n4.Scaled To Fill) {constB=4}
  n7 = Multiply(A=Time(), B=ScalarParameter() {defaultValue=0.1, parameterName=Time Factor})
  n8 = TextureSample(UVs=Panner(Coordinate=@n6, Time=@n7) {speedX=0.35, speedY=0.125}) {samplerSource=SSM_Wrap_WorldGroupSettings, texture=/Game/UI/Frontend/ItemShop/Art/IWDTakeover/Texture/T-Noise-IWD}
  n9 = TextureSample(UVs=Panner(Coordinate=@n6, Time=@n7) {speedX=0.25, speedY=0.25}) {samplerType=SAMPLERTYPE_Grayscale, texture=&LowResBlurredNoise}
  n10 = LinearInterpolate(A=@n5.Result, Alpha=Multiply(A=Add(A=@n8.R, B=@n9.R)) {constB=1.4})
  n11 = MaterialFunctionCall(Texture=@n5.Result, GradientStart=Constant() {r=0.088}, GradientEnd=Constant() {r=1.0594}, RotationAngle=Constant() {r=0.75}) {materialFunction=&MF_UI_LinearGradient}
  n12 = Multiply(A=LinearInterpolate(B=Constant3Vector() {constant=<rgba 0.5251 0 1 0>}, Alpha=LinearInterpolate(B=@n10, Alpha=@n11.Result))) {constB=2}
  n13 = MaterialFunctionCall(GradientStart=Constant() {r=0.25}, GradientEnd=Constant() {r=1}, RotationAngle=Constant() {r=0.75}) {materialFunction=&MF_UI_LinearGradient}
  n14 = MaterialFunctionCall() {materialFunction=&MF_UI_UVScaling}
  n15 = ConstantBiasScale(None=@n14.Scaled To Fill) {bias=-0.5, scale=2}
  n16 = Subtract(A=ComponentMask(None=Abs(None=@n15)) {g=true}, B=ScalarParameter() {defaultValue=0.0511, parameterName=Subtract})
  n17 = ScalarParameter() {defaultValue=1.0721, parameterName=Falloff}
  n18 = Multiply(A=Saturate(None=Multiply(A=@n16, B=@n17)), B=ScalarParameter() {defaultValue=1.085, parameterName=Intensity})
  n19 = ConstantBiasScale(None=@n14.Scaled To Fill) {bias=-0.5, scale=2}
  n20 = Subtract(A=ComponentMask(None=Abs(None=@n19)) {r=true}, B=ScalarParameter() {defaultValue=0.302, parameterName=SubtractH})
  n21 = ScalarParameter() {defaultValue=0.8595, parameterName=FalloffH}
  n22 = Multiply(A=Saturate(None=Multiply(A=@n20, B=@n21)), B=ScalarParameter() {defaultValue=2.3306, parameterName=IntensityH})
  n23 = LinearInterpolate(A=Constant3Vector(), Alpha=OneMinus(None=LinearInterpolate(A=@n18, B=@n22))) {constAlpha=0}
  n24 = TextureSample() {texture=&/T_UI_FNMares2022_StoreBG}
  n25 = SmoothStep(Value=@n3) {constMax=8, constMin=0.5}
  LinearInterpolate(
   A=LinearInterpolate(B=LinearInterpolate(B=@n12, Alpha=@n13.Result), Alpha=TextureSample() {texture=&/T_UI_FNMares2022_Mask}.RGB)
   B=LinearInterpolate(
    A=Multiply(A=@n23, B=@n24.RGB)
    B=TextureSample() {texture=&/T_UI_FNMares2022_StoreBG_Light}.RGB
    Alpha=Saturate(None=@n25)
   ) {constB=0}
  )
 ]
}
